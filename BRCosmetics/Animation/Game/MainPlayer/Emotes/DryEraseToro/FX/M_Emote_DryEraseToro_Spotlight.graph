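# Translucent Niagara mesh-particle material for the spotlight effect in the DryEraseToro emote, with particle color feeding emissive adjustments.
# A soft vertical ramp blends with panning detail: two smoke layers scaled by the Smoke_light dynamic parameter, plus bokeh dots and dust masked by a gradient.
# That mask is scaled by particle alpha, then faded near the camera and by view angle, with depth fade skipped on mobile.
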Material /BRCosmetics/Animation/Game/MainPlayer/Emotes/DryEraseToro/FX/M_Emote_DryEraseToro_Spotlight
graph {
 properties {bUsedWithNiagaraMeshParticles=true, blendMode=BLEND_Translucent, pixelDepthOffsetMode=PDOM_Legacy, usedShadingModels=Default Lit}
 nodes [
  n1 = ParticleColor()
  MaterialFunctionCall(Emissive=@n1.RGB) {materialFunction=&MF_FX_EmissiveAdjustments}
  n2 = ScalarParameter() {defaultValue=200, parameterName=Near Fade}
  n3 = ScalarParameter() {defaultValue=4, parameterName=Angle Fade}
  n4 = TextureSample() {texture=/Game/Effects/Fort_Effects/Textures/Ramps/T_Ramp_Vert_Soft_Fade}
  n5 = TextureSample(UVs=TextureCoordinate() {vTiling=-1}) {texture=&T_Ramp_Vert_Soft_Fade}
  n6 = VectorParameter() {parameterName=Smoke A UV, sortPriority=1, defaultValue=<rgba 2 2 -0.05 0.2>}
  n7 = Panner(Coordinate=Multiply(A=TextureCoordinate(), B=AppendVector(A=@n6.R, B=@n6.R)), Speed=AppendVector(A=@n6.B, B=@n6.B))
  n8 = VectorParameter() {parameterName=Smoke B UV, sortPriority=1, defaultValue=<rgba 1 1 0.05 0.15>}
  n9 = Multiply(A=TextureCoordinate(), B=AppendVector(A=@n8.R, B=@n8.R))
  n10 = AppendVector(A=@n8.B, B=@n8.B)
  n11 = LinearInterpolate(
   A=Power(Base=TextureSample(UVs=@n7) {texture=/Game/Global/Textures/Default/Tileable_Patterns/T_Smoke_Puff_Tile_Up}.R)
   B=TextureSample(UVs=Panner(Coordinate=@n9, Speed=@n10)) {texture=/Game/Global/Textures/Default/Tileable_Patterns/T_Thick_Smoke_05}.R
  )
  n12 = DynamicParameter() {paramNames [Smoke_light,Param2,Param3,Param4]}
  n13 = Multiply(A=Add(A=TextureCoordinate()) {constB=-0.5}, B=Constant2Vector() {g=1, r=1}) {constB=0.2}
  n14 = TextureSample(UVs=Panner(Coordinate=Add(A=@n13) {constB=0.5}) {speedY=0.01}) {
   samplerType=SAMPLERTYPE_LinearColor
   texture=/BRCosmetics/Characters/Player/Male/Medium/Bodies/M_MED_SpeedDialBattle/FX/Textures/T_Bokeh_Dots
  }
  n15 = TextureCoordinate()
  n16 = Add(A=Multiply(A=Add(A=@n15) {constB=-0.5}, B=Constant2Vector() {g=1, r=3}) {constB=0.2}) {constB=0.5}
  n17 = Add(A=@n14.G, B=TextureSample(UVs=Panner(Coordinate=@n16) {speedY=0.005}) {texture=/Game/Global/Textures/Default/PointSplatter/T_Tiling_Dust}.G)
  n18 = Add(
   A=Multiply(A=Power(Base=@n5.G) {constExponent=0.5}, B=Multiply(A=@n11, B=@n12.Smoke_light))
   B=Multiply(
    A=Multiply(A=@n17)
    B=MaterialFunctionCall() {materialFunction=&LinearGradient}.VGradient
   )
  )
  n19 = ScalarParameter() {defaultValue=0.5, parameterName=Lerp Smoke}
  n20 = Multiply(
   A=Multiply(A=@n1.A, B=LinearInterpolate(A=@n4.G, B=@n18, Alpha=@n19) {constAlpha=0.8})
   B=ScalarParameter() {defaultValue=0.25, parameterName=Alpha Mult}
  )
  n21 = DepthFade(FadeDistance=ScalarParameter() {defaultValue=60, parameterName=Depth Fade})
  Multiply(
   A=Multiply(
    A=MaterialFunctionCall(Distance=@n2) {materialFunction=&Near_Camera_Fade}.Result
    B=MaterialFunctionCall("FadeSensitivity (optional)"=@n3) {materialFunction=&FadeBasedOnViewAngle}.Result
   )
   B=ShadingPathSwitch(Default=Multiply(A=@n20, B=@n21), Mobile=@n20) {desc=Migrated from FeatureLevelSwitch}
  )
 ]
}
