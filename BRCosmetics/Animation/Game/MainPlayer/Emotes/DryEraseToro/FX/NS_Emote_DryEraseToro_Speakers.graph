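# Niagara effect for the DryEraseToro emote's speakers, looping randomly timed pulses like pumping sound waves.
# Two MeshRipples emitters and a refraction glow sphere follow skeletal mesh locations, with expanding mesh/sprite ripples shifting blue to yellow over each loop.
# SmokeRingStreaks bursts three spawn groups, using masked values and a visibility tag to vary smoke, streak and ring sprites.
# Time-of-day functions tune color and opacity; three LWE emitters have no renderers.
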NiagaraSystem /BRCosmetics/Animation/Game/MainPlayer/Emotes/DryEraseToro/FX/NS_Emote_DryEraseToro_Speakers
dependencies {
 usedDataInterfaces [
  &NiagaraDataInterfaceCurve
  &NiagaraDataInterfaceColorCurve
  &NiagaraDataInterfaceArrayColor
  &NiagaraDataInterfaceVectorCurve
  &NiagaraDataInterfaceSkeletalMesh
 ]
 usedDynamicInputs [
  &MaskVector2dBySpawnGroup
  &RandomRangeVector
  &RotateVectorByQuaternion
 ]
 usedModules [
  &self:SystemSpawnScript.SystemScriptSource.SystemScriptGraph.NiagaraNodeAssignment_0.SetVariables_#1
  &SystemState
 ]
}
emitters {
 AirglowLWE {rendererClasses [], renderers []}
 MeshRipples {
  emitterUpdateScript [
   &EmitterState(
    Delay First Loop Only=true
    Life Cycle Mode=::NewEnumerator1
    Loop Delay=0.4
    Loop Duration=&RandomRangeFloat(Minimum=0.585, Maximum=0.727)
    Recalculate Duration Each Loop=true
    UseLoopDelay=true
   )
   &SpawnBurst_Instantaneous(Spawn Count=2, Use Spawn Probability=false, UseLoopCountLimit=false)
   &SpawnBurst_Instantaneous(Spawn Count=2, Spawn Time=0.15, Use Spawn Probability=false, UseLoopCountLimit=false) {module=SpawnBurst_Instantaneous003}
   &SpawnBurst_Instantaneous(Spawn Count=2, Spawn Time=0.05, Use Spawn Probability=false, UseLoopCountLimit=false) {module=SpawnBurst_Instantaneous001}
  ]
  particleSpawnScript [
   &InitializeParticle(
    Color=&FNTimeOfDayColor(Night Color=<rgba 0.4 0.4 0.4 1>, Morning Color=<rgba 0.65 0.65 0.65 1>, Day Color=<rgba 1 1 1 1>, Evening Color=<rgba 0.65 0.65 0.65 1>)
    Lifetime=0.25
    Mesh Scale=<0.25 0.25 0.5>
    Mesh Scale Mode=::NewEnumerator4
    Position Offset=<5 0 -2.5>
    Sprite Size Mode=::NewEnumerator3
    Uniform Sprite Size=25
    UsePositionOffset=false
    Write Material Random=true
   )
   &InitialMeshOrientation(Mesh Orientation Mode=::NewEnumerator2, Rotation=<0.25 0 0>, Use Rotation Vector=true) {module=InitialMeshOrientation001}
   &CameraOffset(Camera Offset Amount=2) {module=CameraOffset001}
   &ApplyInitialForces()
  ]
  particleUpdateScript [
   &ParticleState()
   &SkeletalMeshLocation(
    Bone / Socket Index=1
    MaskSpawnGroup=false
    Mesh Sampling Type=::NewEnumerator5
    Position Sampling=::NewEnumerator1
    Skeletal Mesh=$System.Skeletal Mesh
    Socket Sampling Mode=::NewEnumerator5
   )
   &ScaleMeshSize(Scale Factor=&VectorFromFloat(Value=&Add_Float(
      A=1
      B=&FloatFromCurve(
       FloatCurve=DataInterface() {propertyValues {Curve {keys [
           {
            arriveTangent=1.0578
            arriveTangentWeight=0.2235
            interpMode=RCIM_Cubic
            leaveTangent=1.0578
            leaveTangentWeight=0.2235
            tangentMode=RCTM_User
            time=0.05
            value=0
           }
           {
            arriveTangent=5.2964
            arriveTangentWeight=0.9219
            interpMode=RCIM_Cubic
            leaveTangent=5.2964
            leaveTangentWeight=0.9219
            tangentMode=RCTM_User
            time=1
            value=2
           }
          ]}}}
       Scale Curve=0.67
      )
     )))
   &ScaleSpriteSize(Uniform Curve Sprite Scale=DataInterface() {propertyValues {Curve {keys [{tangentMode=RCTM_User, time=0, value=1}, {tangentMode=RCTM_User, time=1, value=1.45}]}}})
   &ScaleColor(
    Scale Alpha=&FloatFromCurve(
     FloatCurve=DataInterface() {propertyValues {Curve {keys [
         {arriveTangent=-1.1111, interpMode=RCIM_Cubic, time=0, value=0}
         <at 0.1 1 RCIM_Cubic>
         {arriveTangent=-0.9656, interpMode=RCIM_Cubic, leaveTangent=-0.9656, tangentMode=RCTM_User, time=1, value=0}
        ]}}}
     Scale Curve=&FNTimeOfDayScalar(Night Scale=0.275, Morning Scale=0.275, Day Scale=0.2, Evening Scale=0.275)
    )
    Scale RGB=&Multiply_VectorByFloat(
     Vector=&MakeVectorFromLinearColorRGB(Color=&ColorFromCurve(
       ColorCurve=DataInterface() {propertyValues {
         AlphaCurve {keys []}
         BlueCurve {keys [
           <at 0 1>
           {tangentMode=RCTM_User, time=0.4905, value=0.0004}
           {arriveTangentWeight=0.1167, interpMode=RCIM_Cubic, leaveTangentWeight=0.8633, tangentMode=RCTM_User, time=1, value=0}
          ]}
         GreenCurve {keys [
           <at 0 0.3997>
           {tangentMode=RCTM_User, time=0.4905, value=0.5703}
           {arriveTangentWeight=0.1167, interpMode=RCIM_Cubic, leaveTangentWeight=0.8633, tangentMode=RCTM_User, time=1, value=0.8929}
          ]}
         RedCurve {keys [
           <at 0 0>
           {tangentMode=RCTM_User, time=0.4905, value=0.186}
           {arriveTangentWeight=0.1167, interpMode=RCIM_Cubic, leaveTangentWeight=0.8633, tangentMode=RCTM_User, time=1, value=1.25}
          ]}
        }}
       CurveIndex=$Emitter.NormalizedLoopAge
      ))
     Float=5
    )
    ScaleA=true
    ScaleRGB=true
   )
   &DynamicMaterialParameters(Module.Index0Param0Write=true, Module.Index0Param1Write=true, Module.Index0Param2Write=true, Module.Index0Param3Write=true)
   &AccelerationForce(
    Acceleration=&MaskVectorBySpawnGroup(Masked Vector=<-600 750 0>, Passthrough Vector=<-224 280 1>)
    Coordinate Space=::Local
   )
   &SolveForcesAndVelocity(Clamp Velocity=false, Limit Acceleration=false)
   &AlignSpriteToMeshOrientation(
    Mesh Orientation Relative Sprite Alignment Vector=<1 0 0>
    Mesh Orientation Relative Sprite Facing Vector:Vector3f=<0 0 1>
    SetAlignment=true
    SetFacing=true
   )
   /SetVariables(Particles.GlowColor=&MakeLinearColorFromVectorAndFloat(
     "Vector (RGB)"=&Multiply_VectorByFloat(
      Vector=&MakeVectorFromLinearColorRGB(Color=&ColorFromCurve(
        ColorCurve=DataInterface() {propertyValues {
          AlphaCurve {keys []}
          BlueCurve {keys [
            <at 0.2415 1>
            {tangentMode=RCTM_User, time=0.4905, value=0.0004}
            {arriveTangentWeight=0.1167, interpMode=RCIM_Cubic, leaveTangentWeight=0.8633, tangentMode=RCTM_User, time=1, value=0}
           ]}
          GreenCurve {keys [
            <at 0.2415 0.3997>
            {tangentMode=RCTM_User, time=0.4905, value=0.5703}
            {arriveTangentWeight=0.1167, interpMode=RCIM_Cubic, leaveTangentWeight=0.8633, tangentMode=RCTM_User, time=1, value=0.8929}
           ]}
          RedCurve {keys [
            <at 0.2415 0>
            {tangentMode=RCTM_User, time=0.4905, value=0.186}
            {arriveTangentWeight=0.1167, interpMode=RCIM_Cubic, leaveTangentWeight=0.8633, tangentMode=RCTM_User, time=1, value=1.25}
           ]}
         }}
        CurveIndex=$Emitter.NormalizedLoopAge
       ))
      Float=3
     )
     "Float (Alpha)"=&FloatFromCurve(
      FloatCurve=DataInterface() {propertyValues {Curve {keys [
          {arriveTangent=-1.1111, interpMode=RCIM_Cubic, time=0, value=0}
          {arriveTangent=-0.1741, interpMode=RCIM_Cubic, leaveTangent=-0.1741, tangentMode=RCTM_User, time=0.1, value=1}
          {arriveTangent=-0.9656, interpMode=RCIM_Cubic, leaveTangent=-0.9656, tangentMode=RCTM_User, time=0.6, value=0}
         ]}}}
      Scale Curve=&FNTimeOfDayScalar(Night Scale=0.3, Morning Scale=0.4, Day Scale=0.5, Evening Scale=0.4)
     )
    )) {isSetParameters=true, module=SetVariables_#2}
  ]
  rendererClasses [&NiagaraMeshRendererProperties, &NiagaraSpriteRendererProperties]
  renderers [
   {rendererClass=&NiagaraMeshRendererProperties, rendererIndex=0}
   {rendererClass=&NiagaraMeshRendererProperties, rendererIndex=1}
   {rendererClass=&NiagaraSpriteRendererProperties, rendererIndex=2}
  ]
 }
 MeshRipples002 {
  emitterUpdateScript [
   &EmitterState(
    Delay First Loop Only=true
    Life Cycle Mode=::NewEnumerator1
    Loop Delay=0.4
    Loop Duration=&RandomRangeFloat(Minimum=0.585, Maximum=0.727)
    Recalculate Duration Each Loop=true
    UseLoopDelay=true
   )
   &SpawnBurst_Instantaneous(Spawn Count=2, Use Spawn Probability=false, UseLoopCountLimit=false)
   &SpawnBurst_Instantaneous(Spawn Count=2, Spawn Time=0.15, Use Spawn Probability=false, UseLoopCountLimit=false) {module=SpawnBurst_Instantaneous003}
   &SpawnBurst_Instantaneous(Spawn Count=2, Spawn Time=0.05, Use Spawn Probability=false, UseLoopCountLimit=false) {module=SpawnBurst_Instantaneous001}
  ]
  particleSpawnScript [
   &InitializeParticle(
    Color=&FNTimeOfDayColor(Night Color=<rgba 0.4 0.4 0.4 1>, Morning Color=<rgba 0.65 0.65 0.65 1>, Day Color=<rgba 1 1 1 1>, Evening Color=<rgba 0.65 0.65 0.65 1>)
    Lifetime=0.25
    Mesh Scale=&MaskVectorBySpawnGroup(Masked Vector=<0.5 0.5 0.5>, Passthrough Vector=<0.25 0.25 0.5>)
    Mesh Scale Mode=::NewEnumerator4
    Position Offset=<5 0 -2.5>
    Sprite Size Mode=::NewEnumerator3
    Uniform Sprite Size=&MaskFloatBySpawnGroup(Masked Float=45, Passthrough Float=25)
    UsePositionOffset=false
    Write Material Random=true
   )
   &CameraOffset()
   &InitialMeshOrientation(Mesh Orientation Mode=::NewEnumerator2, Rotation=<0.25 0 0>, Use Rotation Vector=true) {module=InitialMeshOrientation001}
   &CameraOffset(Camera Offset Amount=2) {module=CameraOffset001}
   &ApplyInitialForces()
  ]
  particleUpdateScript [
   &ParticleState()
   &SkeletalMeshLocation(
    MaskSpawnGroup=false
    Mesh Sampling Type=::NewEnumerator5
    Position Sampling=::NewEnumerator1
    Skeletal Mesh=$System.Skeletal Mesh
    Socket Sampling Mode=::NewEnumerator5
   )
   &ScaleMeshSize(Scale Factor=&VectorFromFloat(Value=&Add_Float(
      A=1
      B=&FloatFromCurve(
       FloatCurve=DataInterface() {propertyValues {Curve {keys [
           {
            arriveTangent=1.0578
            arriveTangentWeight=0.2235
            interpMode=RCIM_Cubic
            leaveTangent=1.0578
            leaveTangentWeight=0.2235
            tangentMode=RCTM_User
            time=0.05
            value=0
           }
           {
            arriveTangent=5.2964
            arriveTangentWeight=0.9219
            interpMode=RCIM_Cubic
            leaveTangent=5.2964
            leaveTangentWeight=0.9219
            tangentMode=RCTM_User
            time=1
            value=2
           }
          ]}}}
       Scale Curve=0.67
      )
     )))
   &ScaleSpriteSize(Uniform Curve Sprite Scale=DataInterface() {propertyValues {Curve {keys [{tangentMode=RCTM_User, time=0, value=1}, {tangentMode=RCTM_User, time=1, value=1.45}]}}})
   &ScaleColor(
    Scale Alpha=&FloatFromCurve(
     FloatCurve=DataInterface() {propertyValues {Curve {keys [
         {arriveTangent=-1.1111, interpMode=RCIM_Cubic, time=0, value=0}
         <at 0.1 1 RCIM_Cubic>
         {arriveTangent=-0.9656, interpMode=RCIM_Cubic, leaveTangent=-0.9656, tangentMode=RCTM_User, time=1, value=0}
        ]}}}
     Scale Curve=&FNTimeOfDayScalar(Night Scale=0.5, Morning Scale=0.5, Day Scale=0.35, Evening Scale=0.5)
    )
    Scale RGB=&Multiply_VectorByFloat(
     Vector=&MakeVectorFromLinearColorRGB(Color=&ColorFromCurve(
       ColorCurve=DataInterface() {propertyValues {
         AlphaCurve {keys []}
         BlueCurve {keys [
           <at 0 1>
           {tangentMode=RCTM_User, time=0.4905, value=0.0004}
           {arriveTangentWeight=0.1167, interpMode=RCIM_Cubic, leaveTangentWeight=0.8633, tangentMode=RCTM_User, time=1, value=0}
          ]}
         GreenCurve {keys [
           <at 0 0.3997>
           {tangentMode=RCTM_User, time=0.4905, value=0.5703}
           {arriveTangentWeight=0.1167, interpMode=RCIM_Cubic, leaveTangentWeight=0.8633, tangentMode=RCTM_User, time=1, value=0.8929}
          ]}
         RedCurve {keys [
           <at 0 0>
           {tangentMode=RCTM_User, time=0.4905, value=0.186}
           {arriveTangentWeight=0.1167, interpMode=RCIM_Cubic, leaveTangentWeight=0.8633, tangentMode=RCTM_User, time=1, value=1.25}
          ]}
        }}
       CurveIndex=$Emitter.NormalizedLoopAge
      ))
     Float=5
    )
    ScaleA=true
    ScaleRGB=true
   )
   &DynamicMaterialParameters(Module.Index0Param0Write=true, Module.Index0Param1Write=true, Module.Index0Param2Write=true, Module.Index0Param3Write=true)
   &AccelerationForce(
    Acceleration=&MaskVectorBySpawnGroup(Masked Vector=<-600 750 0>, Passthrough Vector=<-224 280 1>)
    Coordinate Space=::Local
   )
   &SolveForcesAndVelocity(Clamp Velocity=false, Limit Acceleration=false)
   &AlignSpriteToMeshOrientation(
    Mesh Orientation Relative Sprite Alignment Vector=<1 0 0>
    Mesh Orientation Relative Sprite Facing Vector:Vector3f=<0 0 1>
    SetAlignment=true
    SetFacing=true
   )
   /SetVariables(Particles.GlowColor=&MakeLinearColorFromVectorAndFloat(
     "Vector (RGB)"=&Multiply_VectorByFloat(
      Vector=&MakeVectorFromLinearColorRGB(Color=&ColorFromCurve(
        ColorCurve=DataInterface() {propertyValues {
          AlphaCurve {keys []}
          BlueCurve {keys [
            <at 0.2415 1>
            {tangentMode=RCTM_User, time=0.4905, value=0.0004}
            {arriveTangentWeight=0.1167, interpMode=RCIM_Cubic, leaveTangentWeight=0.8633, tangentMode=RCTM_User, time=1, value=0}
           ]}
          GreenCurve {keys [
            <at 0.2415 0.3997>
            {tangentMode=RCTM_User, time=0.4905, value=0.5703}
            {arriveTangentWeight=0.1167, interpMode=RCIM_Cubic, leaveTangentWeight=0.8633, tangentMode=RCTM_User, time=1, value=0.8929}
           ]}
          RedCurve {keys [
            <at 0.2415 0>
            {tangentMode=RCTM_User, time=0.4905, value=0.186}
            {arriveTangentWeight=0.1167, interpMode=RCIM_Cubic, leaveTangentWeight=0.8633, tangentMode=RCTM_User, time=1, value=1.25}
           ]}
         }}
        CurveIndex=$Emitter.NormalizedLoopAge
       ))
      Float=3
     )
     "Float (Alpha)"=&FloatFromCurve(
      FloatCurve=DataInterface() {propertyValues {Curve {keys [
          {arriveTangent=-1.1111, interpMode=RCIM_Cubic, time=0, value=0}
          {arriveTangent=-0.1741, interpMode=RCIM_Cubic, leaveTangent=-0.1741, tangentMode=RCTM_User, time=0.1, value=1}
          {arriveTangent=-0.9656, interpMode=RCIM_Cubic, leaveTangent=-0.9656, tangentMode=RCTM_User, time=0.6, value=0}
         ]}}}
      Scale Curve=&FNTimeOfDayScalar(Night Scale=0.6, Morning Scale=0.7, Day Scale=0.85, Evening Scale=0.7)
     )
    )) {isSetParameters=true, module=SetVariables_#2}
  ]
  rendererClasses [&NiagaraMeshRendererProperties, &NiagaraSpriteRendererProperties]
  renderers [
   {rendererClass=&NiagaraMeshRendererProperties, rendererIndex=0}
   {rendererClass=&NiagaraMeshRendererProperties, rendererIndex=1}
   {rendererClass=&NiagaraSpriteRendererProperties, rendererIndex=2}
  ]
 }
 NE_RefracGlowSphere {
  emitterUpdateScript [
   &EmitterState(
    Delay First Loop Only=true
    Enable Distance Culling=true
    Life Cycle Mode=::NewEnumerator1
    Loop Delay=0.4
    Loop Duration=&RandomRangeFloat(Minimum=0.4, Maximum=1.2)
    MaxDistance=3500
    Module.bUseMaxDistance=false
    Module.bUseMinDistance=false
    Recalculate Duration Each Loop=true
    Scalability Mode=::NewEnumerator1
    UseLoopDelay=true
   )
   &SpawnBurst_Instantaneous(Use Spawn Probability=false, UseLoopCountLimit=false)
  ]
  particleSpawnScript [&InitializeParticle(
    Color=<rgba 0.1364 0.2344 0.033 2>
    Lifetime=&RandomRangeFloat(Minimum=0.15, Maximum=0.2)
    Mesh Scale Mode=::NewEnumerator3
    Mesh Uniform Scale=0.6
    UsePositionOffset=false
    Write Material Random=true
   ) {module=InitializeParticle001}]
  particleUpdateScript [
   &ParticleState()
   &SkeletalMeshLocation(
    MaskSpawnGroup=false
    Mesh Sampling Type=::NewEnumerator5
    Sampled Position Offset=<0 -15 10>
    Skeletal Mesh=$System.Skeletal Mesh
    Socket Sampling Mode=::NewEnumerator5
   )
   &DynamicMaterialParameters(
    Index 0 Param 1=1.0075
    Index 0 Param 2=&FloatFromCurve(FloatCurve=DataInterface() {propertyValues {Curve {keys [<at 0 15>,<at 1 0>]}}})
    Module.Param0WriteEnabled=true
    Module.Param1WriteEnabled=true
    Module.Param2WriteEnabled=true
    Module.Param3WriteEnabled=true
   )
   &ScaleMeshSize(Scale Factor=/Niagara/DynamicInputs/ScaleByCurve/ScaleVectorByCurve(
     InitialValue:Vector3f=<1 1 1>
     ScaleCurve:NiagaraDataInterfaceVectorCurve=DataInterface() {propertyValues {
       XCurve {keys [
         {arriveTangent=4.2859, interpMode=RCIM_Cubic, leaveTangent=4.2859, tangentMode=RCTM_User, time=0, value=0}
         {arriveTangent=2, interpMode=RCIM_Cubic, leaveTangent=2, tangentMode=RCTM_User, time=0.25, value=1}
         {interpMode=RCIM_Cubic, tangentMode=RCTM_User, time=1, value=2}
        ]}
       YCurve {keys [
         {arriveTangent=5.8758, interpMode=RCIM_Cubic, leaveTangent=5.8759, tangentMode=RCTM_User, time=0, value=0}
         {arriveTangent=2, interpMode=RCIM_Cubic, leaveTangent=2, tangentMode=RCTM_User, time=0.25, value=1}
         {interpMode=RCIM_Cubic, tangentMode=RCTM_User, time=1, value=2}
        ]}
       ZCurve {keys [
         {arriveTangent=4.7007, interpMode=RCIM_Cubic, leaveTangent=4.7007, tangentMode=RCTM_User, time=0, value=0}
         {arriveTangent=2, interpMode=RCIM_Cubic, leaveTangent=2, tangentMode=RCTM_User, time=0.25, value=1}
         {interpMode=RCIM_Cubic, tangentMode=RCTM_User, time=1, value=2}
        ]}
      }}
     CurveIndex:NiagaraFloat=$Particles.NormalizedAge
    ))
   &ScaleColor(
    Scale Alpha=&FloatFromCurve(
     FloatCurve=DataInterface() {propertyValues {Curve {keys [
         {interpMode=RCIM_Cubic, tangentMode=RCTM_User, time=0.0045, value=0.9397}
         {arriveTangent=-1.0036, interpMode=RCIM_Cubic, leaveTangent=-1.0036, tangentMode=RCTM_User, time=0.4647, value=0.5233}
         {arriveTangent=-2.4196, interpMode=RCIM_Cubic, leaveTangent=-2.4196, tangentMode=RCTM_User, time=1.0011, value=0.002}
        ]}}}
     Scale Curve=0.1
    )
    Scale RGB=<0 0 0>
    ScaleA=true
    ScaleRGB=false
   )
   &SolveForcesAndVelocity(Clamp Velocity=false, Limit Acceleration=false)
  ]
  rendererClasses [&NiagaraMeshRendererProperties]
  renderers [{rendererClass=&NiagaraMeshRendererProperties, rendererIndex=0}]
 }
 SmokeLWE {rendererClasses [], renderers []}
 SmokeRingStreaks {
  emitterUpdateScript [
   &EmitterState(
    Delay First Loop Only=true
    Life Cycle Mode=::NewEnumerator1
    Loop Delay=0.4
    Loop Duration=&RandomRangeFloat(Minimum=0.5, Maximum=0.8)
    Recalculate Duration Each Loop=true
    UseLoopDelay=true
   )
   &SpawnBurst_Instantaneous(
    Spawn Count=&RandomRangeInt(Minimum=15, Maximum=24)
    Use Spawn Probability=false
    UseLoopCountLimit=false
   )
   &SpawnBurst_Instantaneous(
    Spawn Count=&RandomRangeInt(Minimum=15, Maximum=24)
    Spawn Group=1
    Use Spawn Probability=false
    UseLoopCountLimit=false
   ) {module=SpawnBurst_Instantaneous001}
   &SpawnBurst_Instantaneous(Spawn Group=2, Use Spawn Probability=false, UseLoopCountLimit=false) {module=SpawnBurst_Instantaneous002}
  ]
  particleSpawnScript [
   /SetVariables(Particles.VisibilityTag=&SpawnGroup()) {isSetParameters=true}
   &InitializeParticle(
    Color=&SelectLinearColorFromArray(
     Array Sampling Mode=::NewEnumerator0
     Direct Array Index=?
     Color Selection Array=DataInterface() {propertyValues {ColorData [<rgba 0.903 0.9271 0.1598 0.05>, <rgba 0.2392 0.9271 0.1932 0.05>, <rgba 0.1336 0.9271 0.8384 0.05>]}}
     Random Seed=0
    )
    Lifetime Max=&MaskFloatBySpawnGroup(
     Masked Float=0.8
     Passthrough Float=&MaskFloatBySpawnGroup(Masked Float=0.4, Passthrough Float=0.8, Spawn Group Mask=1)
    )
    Lifetime Min:NiagaraFloat=&MaskFloatBySpawnGroup(
     Masked Float=0.5
     Passthrough Float=&MaskFloatBySpawnGroup(Masked Float=0.35, Passthrough Float=0.5, Spawn Group Mask=1)
    )
    Lifetime Mode=::NewEnumerator1
    Sprite Rotation Mode=::NewEnumerator2
    Sprite Size Mode=::NewEnumerator3
    Uniform Sprite Size=&RandomRangeFloat(Minimum=15, Maximum=25)
    UsePositionOffset=false
    Write Material Random=true
   )
   &ShapeLocation(
    Cylinder Height=&MaskFloatBySpawnGroup(Masked Float=5, Passthrough Float=0)
    Cylinder Height Midpoint=0
    Cylinder Radius=&MaskFloatBySpawnGroup(
     Masked Float=25
     Passthrough Float=&MaskFloatBySpawnGroup(Masked Float=0.5, Passthrough Float=0, Spawn Group Mask=1)
    )
    MaskSpawnGroup=false
    Override Seed=false
    Shape Primitive=::NewEnumerator1
    Surface Only=true
    Use Endcaps In Surface Only Mode=false
    UseOwnerScale=true
   )
   &AddVelocity(Velocity=&MakeVector(
     X=0
     Y=0
     Z=&UniformRangedFloat(
      Minimum=0
      Maximum=&MaskFloatBySpawnGroup(
       Masked Float=0
       Passthrough Float=&MaskFloatBySpawnGroup(Masked Float=50, Passthrough Float=0, Spawn Group Mask=1)
      )
     )
    )) {module=AddVelocity001}
   &AddVelocity(
    Module.Use Falloff Distance=false
    Override Seed=false
    Velocity Mode=::NewEnumerator1
    Velocity Speed=&RandomRangeFloat(
     Minimum=&MaskFloatBySpawnGroup(
      Masked Float=125
      Passthrough Float=&MaskFloatBySpawnGroup(Masked Float=150, Passthrough Float=0, Spawn Group Mask=1)
     )
     Maximum=&MaskFloatBySpawnGroup(
      Masked Float=200
      Passthrough Float=&MaskFloatBySpawnGroup(Masked Float=150, Passthrough Float=0, Spawn Group Mask=1)
     )
    )
   )
   &CameraOffset(Camera Offset Amount=5)
   /SetVariables(Particles.WindStreakDP:Vector4f=&MakeVector4(
     X=&RandomRangeFloat(Minimum=0, Maximum=1)
     Y=&RandomRangeFloat(Minimum=1, Maximum=2)
     Z=1
     W=1
    )) {isSetParameters=true}
   /SetVariables(Particles.RingAlignment:Vector3f=<0 1 0>, Particles.RingFacing:Vector3f=<0 0 1>) {isSetParameters=true}
  ]
  particleUpdateScript [
   &ParticleState()
   &GravityForce(Gravity=&MakeVector(
     X=0
     Y=0
     Z=&FloatFromCurve(
      FloatCurve=DataInterface() {propertyValues {Curve {keys [<at 0 0 RCIM_Cubic>,<at 0.4 1 RCIM_Cubic>]}}}
      Scale Curve=&MaskFloatBySpawnGroup(Masked Float=&RandomRangeFloat(Minimum=50, Maximum=125), Passthrough Float=0)
     )
    ))
   &Drag(
    Drag=&MaskFloatBySpawnGroup(Masked Float=10, Passthrough Float=1.5)
    Use Linear Drag=true
    Use Rotational Drag=false
   )
   &SolveForcesAndVelocity(Clamp Velocity=false, Limit Acceleration=false)
   &ScaleSpriteSize(Uniform Curve Sprite Scale=DataInterface() {propertyValues {Curve {keys [
        {interpMode=RCIM_Cubic, tangentMode=RCTM_User, time=0, value=0.5}
        {arriveTangent=1.5396, interpMode=RCIM_Cubic, leaveTangent=1.5396, tangentMode=RCTM_User, time=1, value=2}
       ]}}})
   /SetVariables(Particles.SubImageIndex=&FloatFromCurve(FloatCurve=DataInterface() {propertyValues {Curve {defaultValue=5.1927, keys [{tangentMode=RCTM_User, time=0, value=0}, {tangentMode=RCTM_User, time=1, value=1}]}}})) {isSetParameters=true}
   &DynamicMaterialParameters(
    Index 0 Param 1=&FloatFromCurve(FloatCurve=DataInterface() {propertyValues {Curve {keys [<at -0.0046 1.0113 RCIM_Cubic>,<at 1.0011 0.002>]}}})
    Index 0 Param 4=&FloatFromCurve(FloatCurve=DataInterface() {propertyValues {Curve {keys [<at 0 0 RCIM_Cubic>,<at 1 0.5>]}}})
    Module.Param0WriteEnabled=true
    Module.Param1WriteEnabled=false
    Module.Param2WriteEnabled=false
    Module.Param3WriteEnabled=true
   ) {module=DynamicMatParams}
   &ScaleColor(
    Scale Alpha=&FloatFromCurve(
     FloatCurve=DataInterface() {propertyValues {Curve {keys [
         {interpMode=RCIM_Cubic, tangentMode=RCTM_User, time=0, value=0}
         <at 0.05 1>
         {arriveTangent=-1.0493, interpMode=RCIM_Cubic, leaveTangent=-1.0493, time=0.5, value=1}
         <at 1.0011 0.002>
        ]}}}
     Scale Curve=&FNTimeOfDayScalar(Night Scale=1.5, Morning Scale=0.7, Evening Scale=0.7)
    )
    ScaleA=true
    ScaleRGB=false
   )
   /SetVariables(
    Particles.WindStreakColor:LinearColor=&MultiplyLinearColor(
     Linear Color=&RandomRangeLinearColor(Minimum=<rgba 1 0.7958 0 1>, Maximum=<rgba 0.471 1 0 1>)
     Scale Factor=&MakeVector4_FromLinearColor(LinearColor=&MakeLinearColorFromVectorAndFloat(
       "Vector (RGB)"=&MakeVectorFromLinearColorRGB(Color=&FNTimeOfDayColor(
         Night Color=<rgba 0.2123 0.2489 0.27 1>
         Morning Color=<rgba 0.445 0.2629 0.1507 1>
         Day Color=<rgba 0.3626 0.4008 0.4271 1>
         Evening Color=<rgba 0.599 0.2667 0.2028 1>
        ))
       "Float (Alpha)"=&FNTimeOfDayScalar(Night Scale=0.75, Morning Scale=1, Day Scale=0.9, Evening Scale=1.2)
      ))
    )
    Particles.WindStreakSize:Vector2f=&RandomRangeVector2D(Minimum=<3 35>, Maximum=<5 50>)
   ) {isSetParameters=true}
   /SetVariables(
    Particles.RingCO=20
    Particles.RingColor:LinearColor=&MultiplyLinearColor(
     Linear Color=<rgba 0.7555 0.7812 0.3899 1>
     Scale Factor=&MakeVector4_FromLinearColor(LinearColor=&MakeLinearColorFromVectorAndFloat(
       "Vector (RGB)"=&MakeVectorFromLinearColorRGB(Color=&FNTimeOfDayColor(
         Night Color=<rgba 0.2123 0.2489 0.27 1>
         Morning Color=<rgba 0.445 0.2629 0.1507 1>
         Day Color=<rgba 0.3626 0.4008 0.4271 1>
         Evening Color=<rgba 0.599 0.2667 0.2028 1>
        ))
       "Float (Alpha)"=&FloatFromCurve(FloatCurve=DataInterface() {propertyValues {Curve {keys [
            {tangentMode=RCTM_User, time=0, value=1}
            {arriveTangent=-0.0405, interpMode=RCIM_Cubic, leaveTangent=-0.0405, tangentMode=RCTM_User, time=0.7009, value=1}
            {arriveTangent=-8.0089, interpMode=RCIM_Cubic, leaveTangent=-8.0089, tangentMode=RCTM_User, time=1, value=0}
           ]}}})
      ))
    )
    Particles.RingDP:Vector4f=&MakeVector4(
     X=&FloatFromCurve(FloatCurve=DataInterface() {propertyValues {Curve {keys [
          {arriveTangent=4.8499, interpMode=RCIM_Cubic, leaveTangent=4.8499, tangentMode=RCTM_User, time=0, value=0.5}
          {arriveTangent=0.065, interpMode=RCIM_Cubic, leaveTangent=0.065, tangentMode=RCTM_User, time=1, value=2}
         ]}}})
     Y=0
     Z=0
     W=0
    )
    Particles.RingSize:Vector2f=&Vector2DFromFloat(Value=&Multiply_Float(
      A=5
      B=&FloatFromCurve(
       FloatCurve=DataInterface() {propertyValues {Curve {keys [
           {arriveTangent=2.5949, interpMode=RCIM_Cubic, leaveTangent=2.5949, tangentMode=RCTM_User, time=0, value=0}
           {arriveTangent=0.0224, interpMode=RCIM_Cubic, leaveTangent=0.0224, tangentMode=RCTM_User, time=1, value=1}
          ]}}}
       Scale Curve=25
      )
     ))
   ) {isSetParameters=true}
  ]
  rendererClasses [&NiagaraSpriteRendererProperties]
  renderers [
   {rendererClass=&NiagaraSpriteRendererProperties}
   {rendererClass=&NiagaraSpriteRendererProperties, rendererIndex=1}
   {rendererClass=&NiagaraSpriteRendererProperties, rendererIndex=2}
  ]
 }
 SparklesLWE {rendererClasses [], renderers []}
}
summary {
 systemName=NS_Emote_DryEraseToro_Speakers
 emitters [
  {emitterName=MeshRipples002, rendererClasses [&NiagaraMeshRendererProperties, &NiagaraSpriteRendererProperties]}
  {emitterName=SmokeRingStreaks, rendererClasses [&NiagaraSpriteRendererProperties]}
  {emitterName=NE_RefracGlowSphere, rendererClasses [&NiagaraMeshRendererProperties]}
  {emitterName=SparklesLWE, rendererClasses []}
  {emitterName=AirglowLWE, rendererClasses []}
  {emitterName=SmokeLWE, rendererClasses []}
  {emitterName=MeshRipples, rendererClasses [&NiagaraMeshRendererProperties, &NiagaraSpriteRendererProperties]}
 ]
 userVariables [
  User.bIsFrontEnd {type=NiagaraBool}
  User.bIsFrontEndPreview {type=NiagaraBool}
  User.bIsInWater {type=NiagaraBool}
  User.Local Alpha Preview {type=NiagaraFloat, defaultValue=0.75, description=""}
  User.Other Alpha Preview {type=NiagaraFloat, defaultValue=0.25, description=""}
 ]
}
